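# Editor utility widget that lets level designers find floating foliage actors and negatively scaled trees.
# Its button gathers actors of each class in a tolerance table, then multi-line traces vertically to the first landscape or static-mesh hit.
# Actors at or beyond their class's height tolerance above that hit count as floating; a separate check catches negative scales.
# Results print with Bugitgo camera commands; only the floating list is cleared per run.
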EditorUtilityWidgetBlueprint /Game/NeverCook/Blutilities/ForLDs/FloatingActorCheck/FindFloatingActors
graphs [
 CheckNegativeScales {nodes [
   n1 = CheckNegativeScales!() -> (then, Actor:Actor Object Reference) {is=title K2Node_FunctionEntry}
   n2 = BreakVector(InVec=GetActorScale3D!(self:Actor Object Reference=@n1) -> (ReturnValue:Vector))
   > n3 = Branch(Condition="float > float"(A=@n2))
   > n4 = Branch(Condition="float > float"(A=@n2.Y))
   > n5 = Branch(Condition="float > float"(A=@n2.Z))
   > n6 = Set IsValidLocal(IsValidLocal:bool=true) -> (Output_Get:bool)
   Return Node!(execute=[@n7.then, @n6.then], IsValid:bool=Get IsValidLocal() -> (IsValidLocal:bool))
   n7 = Set IsValidLocal(execute=[@n5.else, @n3.else, @n4.else], IsValidLocal=false)
  ]}
 EventGraph {nodes [
   "On Clicked (FindMeInvalidContainers)"() {is=title K2Node_ComponentBoundEvent}
   > Clear(TargetArray:Array of Actor Object References=Get AllFloatingActorsFound() -> (AllFloatingActorsFound:Array of Actor Object References))
   > PrintString(InString="Floating Check Initiated ", bPrintToScreen=false)
   > n1 = For Each Loop(Array:Array of Floating Actor Data Struct Structures=Get FloatingActorsDataArray() -> (FloatingActorsDataArray:Array of Floating Actor Data Struct Structures)) -> (Array Element:Floating Actor Data Struct Structure)
   n2 = Get AllFloatingActorsFound()
   n3 = Branch(execute=@n1.Completed, Condition="integer > integer"(A=Length(TargetArray=@n2)))
   n4 = Length(TargetArray=Get AllFloatingActorsFound())
   > PrintString(InString=Append(A="To String (Integer)"(InInt=@n4), B=" Floating Actors Possibly Found ... "), TextColor=(R=1,G=0.3465,B=0,A=1), Duration=5)
   > n5 = PrintFloatingTrees()
   n6 = Get AllNegativeScaledTreeRefs() -> (AllNegativeScaledTreeRefs:Array of Actor Object References)
   n7 = Branch(execute=[@n5.then, @n9.then], Condition="integer > integer"(A=Length(TargetArray=@n6)))
   n8 = Length(TargetArray=Get AllNegativeScaledTreeRefs())
   > PrintString(
    InString=Append(A="To String (Integer)"(InInt=@n8), B=" Negative Scaled Trees Found")
    bPrintToScreen=false
    TextColor=(R=0.1625,G=1,B=0,A=1)
    Duration=5
   )
   > PrintNegativeScaledTrees()
   PrintString(execute=@n7.else, InString="No Negative Scaled Trees Found!", bPrintToScreen=false, TextColor=(R=0.1625,G=1,B=0,A=1), Duration=5)
   n9 = PrintString(execute=@n3.else, InString="No Floating Actors Found!", bPrintToScreen=false, TextColor=(R=0.1625,G=1,B=0,A=1), Duration=5)
   n10 = Set CurrentData(execute=@n1.LoopBody, CurrentData:Floating Actor Data Struct Structure=@n1) -> (Output_Get:Floating Actor Data Struct Structure)
   > n11 = GetAllActorsOfClass!(
    execute
    WorldContextObject:Object Reference
    ActorClass:Actor Class Reference=Break Floating Actor Data Struct!(FloatingActorDataStruct:"Floating Actor Data Struct Structure (by ref)"=@n10) -> (Actor_2_#1:Actor Class Reference, FloatingOffsetZTolerance_7_#2:float=~) {is=title K2Node_BreakStruct}
   ) -> (then, OutActors:Array of Actor Object References)
   > n12 = For Each Loop(Array:Array of Actor Object References=@n11) -> (Array Element:Actor Object Reference)
   > n13 = LineTraceForLandscapeAndStaticMeshes!(
    execute
    self
    Actor:Actor Object Reference=@n12
    FloatingActorData:Floating Actor Data Struct Structure=Get CurrentData() -> (CurrentData:Floating Actor Data Struct Structure)
   ) -> (then, IsFloating:bool, FoundHit:bool, ActorDisplayName:String, ActorFloatingRef:Actor Object Reference)
   > n14 = Branch(Condition=@n13)
   > n15 = Add(TargetArray:Array of Actor Object References=Get AllFloatingActorsFound(), NewItem:"Actor Object Reference (by ref)"=@n13.ActorFloatingRef) {is=title K2Node_CallArrayFunction}
   n16 = CheckNegativeScales!(execute=[@n15.then, @n14.else], self, Actor:Actor Object Reference=@n12) -> (then, IsValid:bool) {is=title K2Node_CallFunction}
   > n17 = Branch(Condition=@n16)
   Add(execute=@n17.else, TargetArray=Get AllNegativeScaledTreeRefs(), NewItem=@n12)
  ]}
 LineTraceForLandscapeAndStaticMeshes {nodes [
   n1 = LineTraceForLandscapeAndStaticMeshes!() -> (then, Actor:Actor Object Reference, FloatingActorData:Floating Actor Data Struct Structure) {is=title K2Node_FunctionEntry}
   > Set ActorLocal(ActorLocal:Actor Object Reference=@n1) -> (Output_Get:Actor Object Reference)
   > Set DataLocal(DataLocal:Floating Actor Data Struct Structure=@n1.FloatingActorData) -> (Output_Get:Floating Actor Data Struct Structure)
   n2 = Get Actor Location(self=@n1)
   > n3 = Multi Line Trace By Channel!(
    execute
    WorldContextObject:Object Reference
    Start:Vector=vector + vector(A=@n2, B="0,0,768")
    End:Vector=vector + vector(A=@n2, B="0,0,-5000")
    TraceChannel:ETraceTypeQuery Enum=TraceTypeQuery10
    bTraceComplex:bool=true
    ActorsToIgnore:Array of Actor Object References=Make Array("[0]":Actor Object Reference=@n1) -> (Array:Array of Actor Object References)
    DrawDebugType:EDrawDebugTrace Enum=ForDuration
    bIgnoreSelf:bool=true
    TraceColor:Linear Color Structure=(R=1,G=0,B=0,A=1)
    TraceHitColor:Linear Color Structure=(R=0,G=1,B=0,A=1)
    DrawTime:float=5
   ) -> (then, OutHits:Array of Hit Result Structures, ReturnValue:bool)
   > n4 = For Each Loop with Break!(Exec:Exec, Array:Array of Hit Result Structures=@n3, Break:Exec=@n17.then) -> (LoopBody:Exec, Array Element:Hit Result Structure, Array Index:int, Completed:Exec)
   n5 = Branch(execute=@n4.Completed, Condition=Get HitLocal() -> (HitLocal:bool))
   n6 = Get ActorLocal() -> (ActorLocal:Actor Object Reference)
   n7 = float - float(A=BreakVector(InVec=Get Actor Location(self=@n6)).Z, B=BreakVector(InVec=Get HitVectorLocal() -> (HitVectorLocal:Vector)).Z)
   > n8 = Branch(Condition="float < float"(A=@n7))
   n9 = Set IsFloatingLocal(execute=[@n8.then, @n12.else], IsFloatingLocal:bool=false) -> (Output_Get:bool)
   n10 = Get ActorLocal()
   Return Node!(
    execute=[@n5.else, @n13.then, @n9.then]
    IsFloating:bool=Get IsFloatingLocal() -> (IsFloatingLocal:bool)
    FoundHit:bool=Get HitLocal()
    ActorDisplayName:String=GetDisplayName(Object=@n10)
    ActorFloatingRef:Actor Object Reference=@n10
   )
   n11 = Get DataLocal() -> (DataLocal:Floating Actor Data Struct Structure)
   n12 = Branch(
    execute=@n8.else
    Condition="float >= float"(
     A=@n7
     B=Break Floating Actor Data Struct(FloatingActorDataStruct=@n11) -> (FloatingOffsetZTolerance_7_#2=~).FloatingOffsetZTolerance_7_#2
    )
   )
   > n13 = Set IsFloatingLocal(IsFloatingLocal=true)
   n14 = BreakHitResult!(Hit:"Hit Result Structure (by ref)"=@n4) -> (
    bBlockingHit:bool
    bInitialOverlap:bool
    Time:float
    Distance:float
    Location:Vector
    ImpactPoint:Vector
    Normal:Vector
    ImpactNormal:Vector
    PhysMat:Physical Material Object Reference
    HitActor:Actor Object Reference
    HitComponent:Primitive Component Object Reference
    HitBoneName:Name
    BoneName:Name
    HitItem:int
    ElementIndex:int
    FaceIndex:int
    TraceStart:Vector
    TraceEnd:Vector
   )
   n15 = Get Class(Object=@n14.HitActor)
   n16 = Branch(execute=@n4.LoopBody, Condition="Equal (Class)"(A=@n15, B=/Script/Landscape.LandscapeStreamingProxy))
   Set HitVectorLocal(execute=[@n22.then, @n21.then, @n20.then, @n16.then, @n18.then, @n19.then], HitVectorLocal:Vector=@n14.Location) -> (Output_Get:Vector)
   > n17 = Set HitLocal(HitLocal:bool=true) -> (Output_Get:bool)
   n18 = Branch(execute=@n16.else, Condition="Equal (Class)"(A=@n15, B=/Script/Landscape))
   n19 = Branch(execute=@n18.else, Condition="Equal (Class)"(A=@n15, B=/Script/FortniteGame.FortStaticMeshActor))
   n20 = Branch(execute=@n19.else, Condition="Equal (Class)"(A=@n15, B=&StaticMesh))
   n21 = Branch(execute=@n20.else, Condition="Equal (Class)"(A=@n15, B=/Game/Athena/Artemis/Environments/Blueprints/BP_StaticEnvActor.BP_StaticEnvActor_C))
   n22 = Cast To BuildingStaticGeo!(execute=@n21.else, Object=@n14.HitActor) -> (then, CastFailed, "AsBuilding Static Mesh (Static Geo)":"Building Static Mesh (Static Geo) Object Reference") {is=title K2Node_DynamicCast}
  ]}
 PrintFloatingTrees {nodes [
   PrintFloatingTrees() {is=title K2Node_FunctionEntry}
   > n1 = For Each Loop(Array=Get AllFloatingActorsFound())
   n2 = Get Actor Location(self=@n1)
   n3 = BreakVector(InVec=@n2)
   n4 = float + float(A=@n3.Z, B=256)
   n5 = vector + vector(A=@n2, B="256,0,256")
   n6 = BreakRotator(InRot=FindLookAtRotation!(Start:"Vector (by ref)"=@n5, Target:"Vector (by ref)"=@n2) -> (ReturnValue:Rotator))
   n7 = BreakVector(InVec=@n5)
   > PrintString(
    InString=Append!(
     A=GetDisplayName(Object=@n1)
     B="  //  "
     C:String="Bugitgo "
     D:String="To String (Float)"(InDouble=@n7)
     E:String=" "
     F:String="To String (Float)"(InDouble=@n3.Y)
     G:String=" "
     H:String="To String (Float)"(InDouble=@n4)
     I:String=" "
     J:String="To String (Float)"(InDouble=@n6.Pitch)
     K:String="  "
     L:String="To String (Float)"(InDouble=@n6.Yaw)
     M:String=" "
     N:String="To String (Float)"(InDouble=@n6)
    ) -> (ReturnValue)
    bPrintToScreen=false
    TextColor=(R=1,G=0.7058,B=0,A=1)
    Duration=5
   )
  ]}
 PrintNegativeScaledTrees {nodes [
   PrintNegativeScaledTrees() {is=title K2Node_FunctionEntry}
   > n1 = For Each Loop(Array=Get AllNegativeScaledTreeRefs())
   n2 = Get Actor Location(self=@n1)
   n3 = BreakVector(InVec=@n2)
   n4 = float + float(A=@n3.Z, B=256)
   n5 = vector + vector(A=@n2, B="256,0,256")
   n6 = BreakRotator(InRot=FindLookAtRotation(Start=@n5, Target=@n2))
   n7 = BreakVector(InVec=@n5)
   > PrintString(
    InString=Append(
     A=GetDisplayName(Object=@n1)
     B="  //  "
     C="Bugitgo "
     D="To String (Float)"(InDouble=@n7)
     E=" "
     F="To String (Float)"(InDouble=@n3.Y)
     G=" "
     H="To String (Float)"(InDouble=@n4)
     I=" "
     J="To String (Float)"(InDouble=@n6.Pitch)
     K="  "
     L="To String (Float)"(InDouble=@n6.Yaw)
     M=" "
     N="To String (Float)"(InDouble=@n6)
    )
    bPrintToScreen=false
    TextColor=(R=1,G=0.7058,B=0,A=1)
    Duration=5
   )
  ]}
]
variables [
 AllFloatingActorsFound {type="array<Actor>"}
 AllNegativeScaledTreeRefs {type="array<Actor>"}
 CurrentData {type=FloatingActorDataStruct, default {actor=null, floatingOffsetZTolerance=16}}
 FloatingActorsDataArray {
  type="array<FloatingActorDataStruct>"
  default [
   {
    actor=/Hera/Foliage/Plants/Alien_Cactus/Blueprints/BP_Hera_Plant_Alien_Cactus_A_A.BP_Hera_Plant_Alien_Cactus_A_A_C
    floatingOffsetZTolerance=8
   }
   {actor=/Hera/Foliage/Plants/Rosette/Blueprints/BP_Hera_Plant_Rosette_A_A.BP_Hera_Plant_Rosette_A_A_C, floatingOffsetZTolerance=10}
   {actor=/Hera/Foliage/Plants/Rubber/Blueprints/BP_Hera_Plant_Rubber_A_B.BP_Hera_Plant_Rubber_A_B_C, floatingOffsetZTolerance=8}
   {actor=/Hera/Foliage/Trees/Rosette/Blueprints/BP_Hera_Rosette_Tree_A_A.BP_Hera_Rosette_Tree_A_A_C, floatingOffsetZTolerance=32}
   {actor=/Hera/Foliage/Trees/Rosette/Blueprints/BP_Hera_Rosette_Tree_A_C.BP_Hera_Rosette_Tree_A_C_C, floatingOffsetZTolerance=32}
   {actor=/Hera/Foliage/Trees/Rubber/Blueprints/BP_Hera_RubberTree_A_A.BP_Hera_RubberTree_A_A_C, floatingOffsetZTolerance=12}
   {actor=/Hera/Foliage/Trees/Rubber/Blueprints/BP_Hera_RubberTree_A_B.BP_Hera_RubberTree_A_B_C, floatingOffsetZTolerance=12}
   {
    actor=/Game/Environments/Asteria/Foliage/Trees/BorealConifer_PineTree/Blueprint/BP_Hera_BorealConifer_PineTree_A_B_Boreal.BP_Hera_BorealConifer_PineTree_A_B_Boreal_C
    floatingOffsetZTolerance=35
   }
   {actor=/Hermes/Foliage/Trees/Cedar/Blueprints/BP_Hera_Cedar_Tree_A_B_Boreal.BP_Hera_Cedar_Tree_A_B_Boreal_C, floatingOffsetZTolerance=75}
   {actor=/Hermes/Foliage/Trees/Cedar/Blueprints/BP_Hera_Cedar_Tree_B_B_Boreal.BP_Hera_Cedar_Tree_B_B_Boreal_C, floatingOffsetZTolerance=75}
   {actor=/Hera/Foliage/Plants/Fern/Blueprints/BP_Hera_Giant_Fern_A_A.BP_Hera_Giant_Fern_A_A_C, floatingOffsetZTolerance=16}
   {actor=/Hera/Foliage/Plants/Manzanita/Blueprints/BP_Hera_Manzanita_A_A.BP_Hera_Manzanita_A_A_C, floatingOffsetZTolerance=24}
   {
    actor=/Game/Environments/Asteria/Foliage/Trees/Pine/Blueprint/BP_Hera_Pine_Large_A_A_Boreal.BP_Hera_Pine_Large_A_A_Boreal_C
    floatingOffsetZTolerance=38
   }
   {
    actor=/Game/Environments/Asteria/Foliage/Trees/Pine/Blueprint/BP_Hera_Pine_Small_A_A_Boreal.BP_Hera_Pine_Small_A_A_Boreal_C
    floatingOffsetZTolerance=16
   }
   {
    actor=/Game/Environments/Asteria/Foliage/ChaparralPlants/JacobsLadder/Blueprints/BP_Hera_Plant_JacobsLadder_A.BP_Hera_Plant_JacobsLadder_A_C
    floatingOffsetZTolerance=8
   }
   {actor=/Hermes/Foliage/Trees/Saplings/Blueprints/BP_Hermes_Sapling_Tree_A_D.BP_Hermes_Sapling_Tree_A_D_C, floatingOffsetZTolerance=12}
   {actor=/Hermes/Foliage/Trees/Saplings/Blueprints/BP_Hermes_Sapling_Tree_B_C.BP_Hermes_Sapling_Tree_B_C_C, floatingOffsetZTolerance=12}
   {actor=/Hermes/Foliage/Trees/Saplings/Blueprints/BP_Hermes_Sapling_Tree_C_C.BP_Hermes_Sapling_Tree_C_C_C, floatingOffsetZTolerance=12}
   {actor=/Game/Environments/Helios/Foliage/Plants/CapeAloe/Blueprints/BP_Helios_CapeAloe.BP_Helios_CapeAloe_C, floatingOffsetZTolerance=8}
   {
    actor=/Game/Environments/Helios/Foliage/Plants/CapeAloeFlower/Blueprints/BP_Helios_CapeAloeFlower.BP_Helios_CapeAloeFlower_C
    floatingOffsetZTolerance=8
   }
   {
    actor=/Game/Environments/Helios/Foliage/Plants/HedgehogCactus_A/Blueprints/BP_Helios_HedgehogCactusFlower_A.BP_Helios_HedgehogCactusFlower_A_C
    floatingOffsetZTolerance=15
   }
   {
    actor=/Game/Environments/Helios/Foliage/Plants/PipeCactus_B/Blueprints/BP_Helios_PipeCactus_B.BP_Helios_PipeCactus_B_C
    floatingOffsetZTolerance=8
   }
   {
    actor=/Game/Environments/Helios/Foliage/Plants/PipeCactusFlower_A/Blueprints/BP_Helios_PipeCactusFlower_A.BP_Helios_PipeCactusFlower_A_C
    floatingOffsetZTolerance=10
   }
   {
    actor=/Game/Environments/Helios/Foliage/Plants/PygmyPalm_A/Blueprints/BP_Helios_PygmyPalm_A.BP_Helios_PygmyPalm_A_C
    floatingOffsetZTolerance=8
   }
   {
    actor=/Game/Environments/Helios/Foliage/Plants/PygmyPalm_B/Blueprints/BP_Helios_PygmyPalm_B.BP_Helios_PygmyPalm_B_C
    floatingOffsetZTolerance=8
   }
   {actor=/Hera/Foliage/Trees/Beach_Palm/Blueprints/BP_Hera_BeachPalm_Tree_A.BP_Hera_BeachPalm_Tree_A_C, floatingOffsetZTolerance=45}
   {actor=/Hera/Foliage/Trees/Beach_Palm/Blueprints/BP_Hera_BeachPalm_Tree_B.BP_Hera_BeachPalm_Tree_B_C, floatingOffsetZTolerance=45}
   {
    actor=/Game/Environments/Helios/Foliage/Trees/DesertPalm_A/Blueprint/BP_Hera_DesertPalm_A_C.BP_Hera_DesertPalm_A_C_C
    floatingOffsetZTolerance=20
   }
   {
    actor=/Game/Environments/Helios/Foliage/Trees/DesertPalm_B/Blueprint/BP_Hera_DesertPalm_B_C.BP_Hera_DesertPalm_B_C_C
    floatingOffsetZTolerance=20
   }
   {
    actor=/Hermes/Foliage/Trees/Golden_Pine/Blueprints/BP_Hera_GoldenPine_Tree_C_Chaparral.BP_Hera_GoldenPine_Tree_C_Chaparral_C
    floatingOffsetZTolerance=45
   }
   {
    actor=/Game/Environments/Asteria/Foliage/Trees/HolmOak/Blueprints/BP_Hera_HolmOak_Tree_A_B_Chaparral_CS.BP_Hera_HolmOak_Tree_A_B_Chaparral_CS_C
    floatingOffsetZTolerance=80
   }
   {actor=&BP_Hera_Manzanita_A_A_C, floatingOffsetZTolerance=25}
   {
    actor=/Game/Environments/Asteria/Foliage/Trees/Oak/Blueprint/BP_Hera_Oak_Tree_Medium_Grasslands_CS.BP_Hera_Oak_Tree_Medium_Grasslands_CS_C
    floatingOffsetZTolerance=32
   }
   {
    actor=/Game/Environments/Asteria/Foliage/Trees/SessileOak/Blueprint/BP_Tree_Hera_SessileOak_A_B_Grasslands_CS.BP_Tree_Hera_SessileOak_A_B_Grasslands_CS_C
    floatingOffsetZTolerance=80
   }
   {
    actor=/Game/Environments/Asteria/Foliage/Trees/CypressTree/Blueprints/BP_Tree_Asteria_CpyressTreeS29_CS.BP_Tree_Asteria_CpyressTreeS29_CS_C
    floatingOffsetZTolerance=25
   }
   {
    actor=/Game/Environments/Helios/Foliage/Trees/Ironwood_Tree/Blueprints/BP_Helios_Ironwood_Tree.BP_Helios_Ironwood_Tree_C
    floatingOffsetZTolerance=80
   }
   {
    actor=/Game/Environments/Helios/Foliage/Plants/YuccaPlantFlower/Blueprints/BP_Helios_YuccaPlantFlower.BP_Helios_YuccaPlantFlower_C
    floatingOffsetZTolerance=18
   }
   {
    actor=/Game/Environments/Helios/Foliage/Plants/YuccaPlant/Blueprints/BP_Helios_YuccaPlant.BP_Helios_YuccaPlant_C
    floatingOffsetZTolerance=18
   }
   {actor=&BP_Helios_YuccaPlant_C, floatingOffsetZTolerance=1}
   {
    actor=/Game/Environments/Helios/Foliage/Trees/Acacia_Tree/Blueprints/BP_Hera_Acacia_Tree_A_B_Desert_CS.BP_Hera_Acacia_Tree_A_B_Desert_CS_C
    floatingOffsetZTolerance=32
   }
   {
    actor=/Hermes/Foliage/Trees/Black_Pine/Blueprints/BP_Hera_BlackPine_Tree_A_B_Desert.BP_Hera_BlackPine_Tree_A_B_Desert_C
    floatingOffsetZTolerance=48
   }
   {
    actor=/Hermes/Foliage/Trees/Cedar/Blueprints/BP_Hera_Cedar_Tree_A_C_Desert_CS.BP_Hera_Cedar_Tree_A_C_Desert_CS_C
    floatingOffsetZTolerance=48
   }
   {actor=/Hera/Foliage/Plants/Desert_Cactus/Blueprints/BP_Hera_Desert_Cactus_A_A.BP_Hera_Desert_Cactus_A_A_C, floatingOffsetZTolerance=8}
   {actor=/Hera/Foliage/Plants/Desert_Cactus/Blueprints/BP_Hera_Desert_Cactus_A_B.BP_Hera_Desert_Cactus_A_B_C, floatingOffsetZTolerance=8}
   {
    actor=/Hermes/Foliage/Trees/Larch/Blueprints/BP_Hera_Larch_Tree_B_E_Desert_CS.BP_Hera_Larch_Tree_B_E_Desert_CS_C
    floatingOffsetZTolerance=35
   }
   {actor=/Hera/Foliage/Plants/Desert_Cactus/Blueprints/BP_Hera_Desert_Cactus_B_A.BP_Hera_Desert_Cactus_B_A_C, floatingOffsetZTolerance=8}
   {actor=/Hera/Foliage/Plants/Desert_Cactus/Blueprints/BP_Hera_Desert_Cactus_B_B.BP_Hera_Desert_Cactus_B_B_C, floatingOffsetZTolerance=8}
   {
    actor=/Hera/Foliage/Plants/Desert_Palm_Plant/Blueprints/BP_Hera_Desert_Palm_Plant_A_A.BP_Hera_Desert_Palm_Plant_A_A_C
    floatingOffsetZTolerance=20
   }
   {
    actor=/Hera/Foliage/Plants/Desert_Palm_Plant/Blueprints/BP_Hera_Desert_Palm_Plant_A_B.BP_Hera_Desert_Palm_Plant_A_B_C
    floatingOffsetZTolerance=18
   }
   {actor=/Hera/Foliage/Trees/Desert_Palm/Blueprints/BP_Hera_DesertPalm_Tree_A.BP_Hera_DesertPalm_Tree_A_C, floatingOffsetZTolerance=32}
   {actor=/Hera/Foliage/Trees/Desert_Palm/Blueprints/BP_Hera_DesertPalm_Tree_B.BP_Hera_DesertPalm_Tree_B_C, floatingOffsetZTolerance=32}
   {
    actor=/Hermes/Foliage/Trees/Black_Pine/Blueprints/BP_Hera_BlackPine_Tree_A_C_Grasslands.BP_Hera_BlackPine_Tree_A_C_Grasslands_C
    floatingOffsetZTolerance=48
   }
   {
    actor=&BP_Hera_GoldenPine_Tree_C_Chaparral_C
    floatingOffsetZTolerance=35
   }
   {
    actor=/Hermes/Foliage/Trees/Golden_Pine/Blueprints/BP_Hera_GoldenPine_Tree_D_Grasslands.BP_Hera_GoldenPine_Tree_D_Grasslands_C
    floatingOffsetZTolerance=35
   }
   {
    actor=/Game/Environments/Asteria/Foliage/Trees/HolmOak/Blueprints/BP_Hera_HolmOak_Tree_A_C_Grasslands_CS.BP_Hera_HolmOak_Tree_A_C_Grasslands_CS_C
    floatingOffsetZTolerance=85
   }
   {
    actor=&BP_Hera_Oak_Tree_Medium_Grasslands_CS_C
    floatingOffsetZTolerance=24
   }
   {
    actor=/Game/Environments/Asteria/Foliage/Trees/BorealConifer_PineTree/Blueprint/BP_BorealConifer_PineTree_Snow.BP_BorealConifer_PineTree_Snow_C
    floatingOffsetZTolerance=40
   }
   {actor=/Game/Environments/Asteria/Foliage/Trees/Pine/Blueprint/BP_Pine_Large_A_A_Snow.BP_Pine_Large_A_A_Snow_C, floatingOffsetZTolerance=32}
   {actor=/Game/Environments/Asteria/Foliage/Trees/Pine/Blueprint/BP_Pine_Small_A_A_Snow.BP_Pine_Small_A_A_Snow_C, floatingOffsetZTolerance=12}
   {actor=/Game/Athena/Apollo/Environments/BuildingActors/Foliage/Apollo_BigBush.Apollo_BigBush_C, floatingOffsetZTolerance=80}
   {actor=/Hermes/Foliage/Plants/Fern/Blueprints/BP_Hera_Foliage_Fern_A_A.BP_Hera_Foliage_Fern_A_A_C, floatingOffsetZTolerance=8}
   {
    actor=/Game/Environments/Asteria/Foliage/BorealPlants/BorealSapling/Blueprints/BP_Helios_BorPlant_SaplingPine.BP_Helios_BorPlant_SaplingPine_C
    floatingOffsetZTolerance=8
   }
   {actor=/Hera/Foliage/Trees/Umbrella/Blueprints/BP_Hera_Umbrella_Tree_Orange.BP_Hera_Umbrella_Tree_Orange_C, floatingOffsetZTolerance=32}
   {actor=/Hera/Foliage/Trees/BanyanTree/Blueprints/BP_Hera_Banyan_Tree_A.BP_Hera_Banyan_Tree_A_C, floatingOffsetZTolerance=192}
   {actor=/Hera/Foliage/Trees/White_Oak/Blueprints/BP_Hera_WhiteOak_Tree_Gold.BP_Hera_WhiteOak_Tree_Gold_C, floatingOffsetZTolerance=64}
   {actor=/Hera/Foliage/Trees/AlbiziaTree/Blueprints/BP_Hera_Albizia_Tree_A.BP_Hera_Albizia_Tree_A_C, floatingOffsetZTolerance=16}
   {
    actor=/Game/Environments/Asteria/Foliage/Trees/Oak/Blueprint/BP_Hera_Oak_Tree_Medium_DarkNexus_CS.BP_Hera_Oak_Tree_Medium_DarkNexus_CS_C
    floatingOffsetZTolerance=16
   }
   {
    actor=/Game/Environments/Helios/Foliage/Trees/Acacia_Tree/Blueprints/BP_Helios_Acacia_Tree.BP_Helios_Acacia_Tree_C
    floatingOffsetZTolerance=32
   }
   {actor=/Hera/Foliage/Trees/PlumeriaTree/Blueprints/BP_Hera_Plumeria_Tree_A.BP_Hera_Plumeria_Tree_A_C, floatingOffsetZTolerance=16}
   {actor=/Hera/Foliage/Trees/PlumeriaTree/Blueprints/BP_Hera_Plumeria_Tree_B.BP_Hera_Plumeria_Tree_B_C, floatingOffsetZTolerance=16}
   {
    actor=/Hera/Foliage/Trees/CoconotPalmTree/Blueprints/BP_Hera_CoconutPalm_Tree_A_A.BP_Hera_CoconutPalm_Tree_A_A_C
    floatingOffsetZTolerance=48
   }
   {
    actor=/Hera/Foliage/Trees/CoconotPalmTree/Blueprints/BP_Hera_CoconutPalm_Tree_B_A.BP_Hera_CoconutPalm_Tree_B_A_C
    floatingOffsetZTolerance=48
   }
   {
    actor=/Hera/Foliage/Trees/CoconotPalmTree/Blueprints/BP_Hera_CoconutPalm_Tree_C_A.BP_Hera_CoconutPalm_Tree_C_A_C
    floatingOffsetZTolerance=48
   }
   {actor=/Hera/Foliage/Trees/BalsaTree/Blueprints/BP_Hera_Balsa_Tree_A.BP_Hera_Balsa_Tree_A_C, floatingOffsetZTolerance=64}
   {actor=/Hera/Foliage/Trees/BananaTree/Blueprints/BP_Hera_Banana_Tree_A_A.BP_Hera_Banana_Tree_A_A_C, floatingOffsetZTolerance=2}
   {actor=null, floatingOffsetZTolerance=16}
   {actor=/Hera/Foliage/Trees/BananaTree/Blueprints/BP_Hera_Banana_Tree_B_A_A.BP_Hera_Banana_Tree_B_A_A_C, floatingOffsetZTolerance=2}
   {
    actor=&BP_Helios_Ironwood_Tree_C
    floatingOffsetZTolerance=64
   }
   {
    actor=/Game/Environments/Asteria/Foliage/Trees/Birch/Blueprint/BP_Asteria_Birch_Tree_Yellow.BP_Asteria_Birch_Tree_Yellow_C
    floatingOffsetZTolerance=16
   }
   {
    actor=/Game/Environments/Asteria/Foliage/Trees/Birch/Blueprint/BP_Asteria_Birch_Tree_Red.BP_Asteria_Birch_Tree_Red_C
    floatingOffsetZTolerance=16
   }
   {
    actor=/Game/Environments/Asteria/Foliage/Trees/Birch/Blueprint/BP_Helios_Birch_Tree_Red.BP_Helios_Birch_Tree_Red_C
    floatingOffsetZTolerance=16
   }
   {
    actor=/Game/Environments/Asteria/Foliage/Trees/Birch/Blueprint/BP_Hermes_Birch_Tree_Orange.BP_Hermes_Birch_Tree_Orange_C
    floatingOffsetZTolerance=16
   }
   {actor=/Hera/Foliage/Trees/Oak/Blueprints/BP_Hera_Oak_Tree_Large_Yellow.BP_Hera_Oak_Tree_Large_Yellow_C, floatingOffsetZTolerance=48}
  ]
 }
 HitLocation {type=Vector}
]
widgets {
 info {rootWidgetClass=&VerticalBox, widgetCount=6}
 widgets [
  VerticalBox_0 {widgetClassPath=&VerticalBox, widget=VerticalBox_0}
  AlignAllStaticMeshesLocation {
   widgetClassPath=&Button
   parent=VerticalBox_0
   widget=AlignAllStaticMeshesLocation
   properties {widgetStyle {
     clickedSlateSound {resourceObject=null}
     disabled {
      bIsDynamicallyLoaded=false
      drawAs=RoundedBox
      imageType=NoImage
      mirroring=NoMirror
      resourceName=None
      resourceObject=null
      tiling=NoTile
      imageSize=<0 0>
      margin=<ltrb 0 0 0 0>
      outlineSettings {
       bUseBrushTransparency=false
       roundingType=FixedRadius
       width=1
       color {colorUseRule=UseColor_ColorTable, specifiedColor=<rgba 0 0 0 0>}
       cornerRadii=<4 4 4 4>
      }
      tintColor {colorUseRule=UseColor_Specified, specifiedColor=<rgba 0.0395 0.0395 0.0395 1>}
      uVRegion {bIsValid=false, max=<0 0>, min=<0 0>}
     }
     disabledForeground {colorUseRule=UseColor_ColorTable, specifiedColor=<rgba 0 0 0 0>}
     hovered {
      bIsDynamicallyLoaded=false
      drawAs=RoundedBox
      imageType=NoImage
      mirroring=NoMirror
      resourceName=None
      resourceObject=null
      tiling=NoTile
      imageSize=<0 0>
      margin=<ltrb 0 0 0 0>
      outlineSettings {
       bUseBrushTransparency=false
       roundingType=FixedRadius
       width=1
       color {colorUseRule=UseColor_ColorTable, specifiedColor=<rgba 0 0 0 0>}
       cornerRadii=<4 4 4 4>
      }
      tintColor {colorUseRule=UseColor_Specified, specifiedColor=<rgba 0.0953 0.0953 0.0953 1>}
      uVRegion {bIsValid=false, max=<0 0>, min=<0 0>}
     }
     hoveredForeground {colorUseRule=UseColor_ColorTable, specifiedColor=<rgba 0 0 0 0>}
     hoveredSlateSound {resourceObject=null}
     normal {
      bIsDynamicallyLoaded=false
      drawAs=RoundedBox
      imageType=NoImage
      mirroring=NoMirror
      resourceName=None
      resourceObject=null
      tiling=NoTile
      imageSize=<0 0>
      margin=<ltrb 0 0 0 0>
      outlineSettings {
       bUseBrushTransparency=false
       roundingType=FixedRadius
       width=1
       color {colorUseRule=UseColor_ColorTable, specifiedColor=<rgba 0 0 0 0>}
       cornerRadii=<4 4 4 4>
      }
      tintColor {colorUseRule=UseColor_Specified, specifiedColor=<rgba 0.0395 0.0395 0.0395 1>}
      uVRegion {bIsValid=false, max=<0 0>, min=<0 0>}
     }
     normalForeground {colorUseRule=UseColor_ColorTable, specifiedColor=<rgba 0 0 0 0>}
     normalPadding=<ltrb 12 1.5 12 1.5>
     pressed {
      bIsDynamicallyLoaded=false
      drawAs=RoundedBox
      imageType=NoImage
      mirroring=NoMirror
      resourceName=None
      resourceObject=null
      tiling=NoTile
      imageSize=<0 0>
      margin=<ltrb 0 0 0 0>
      outlineSettings {
       bUseBrushTransparency=false
       roundingType=FixedRadius
       width=1
       color {colorUseRule=UseColor_ColorTable, specifiedColor=<rgba 0 0 0 0>}
       cornerRadii=<4 4 4 4>
      }
      tintColor {colorUseRule=UseColor_Specified, specifiedColor=<rgba 0.0284 0.0284 0.0284 1>}
      uVRegion {bIsValid=false, max=<0 0>, min=<0 0>}
     }
     pressedForeground {colorUseRule=UseColor_ColorTable, specifiedColor=<rgba 0 0 0 0>}
     pressedPadding=<ltrb 12 2 12 0>
     pressedSlateSound {resourceObject=null}
    }}
   slot {class=VerticalBoxSlot, properties {horizontalAlignment=HAlign_Left, padding=<ltrb 5 5 5 5>}}
  }
  FindMeInvalidContainers {
   widgetClassPath=&Button
   parent=AlignAllStaticMeshesLocation
   widget=FindMeInvalidContainers
   properties {widgetStyle {
     clickedSlateSound {resourceObject=null}
     disabled {
      bIsDynamicallyLoaded=false
      drawAs=RoundedBox
      imageType=NoImage
      mirroring=NoMirror
      resourceName=None
      resourceObject=null
      tiling=NoTile
      imageSize=<0 0>
      margin=<ltrb 0 0 0 0>
      outlineSettings {
       bUseBrushTransparency=false
       roundingType=FixedRadius
       width=1
       color {colorUseRule=UseColor_ColorTable, specifiedColor=<rgba 0 0 0 0>}
       cornerRadii=<4 4 4 4>
      }
      tintColor {colorUseRule=UseColor_Specified, specifiedColor=<rgba 0.0395 0.0395 0.0395 1>}
      uVRegion {bIsValid=false, max=<0 0>, min=<0 0>}
     }
     disabledForeground {colorUseRule=UseColor_ColorTable, specifiedColor=<rgba 0 0 0 0>}
     hovered {
      bIsDynamicallyLoaded=false
      drawAs=RoundedBox
      imageType=NoImage
      mirroring=NoMirror
      resourceName=None
      resourceObject=null
      tiling=NoTile
      imageSize=<0 0>
      margin=<ltrb 0 0 0 0>
      outlineSettings {
       bUseBrushTransparency=false
       roundingType=FixedRadius
       width=1
       color {colorUseRule=UseColor_ColorTable, specifiedColor=<rgba 0 0 0 0>}
       cornerRadii=<4 4 4 4>
      }
      tintColor {colorUseRule=UseColor_Specified, specifiedColor=<rgba 0.0953 0.0953 0.0953 1>}
      uVRegion {bIsValid=false, max=<0 0>, min=<0 0>}
     }
     hoveredForeground {colorUseRule=UseColor_ColorTable, specifiedColor=<rgba 0 0 0 0>}
     hoveredSlateSound {resourceObject=null}
     normal {
      bIsDynamicallyLoaded=false
      drawAs=RoundedBox
      imageType=NoImage
      mirroring=NoMirror
      resourceName=None
      resourceObject=null
      tiling=NoTile
      imageSize=<0 0>
      margin=<ltrb 0 0 0 0>
      outlineSettings {
       bUseBrushTransparency=false
       roundingType=FixedRadius
       width=1
       color {colorUseRule=UseColor_ColorTable, specifiedColor=<rgba 0 0 0 0>}
       cornerRadii=<4 4 4 4>
      }
      tintColor {colorUseRule=UseColor_Specified, specifiedColor=<rgba 0.0395 0.0395 0.0395 1>}
      uVRegion {bIsValid=false, max=<0 0>, min=<0 0>}
     }
     normalForeground {colorUseRule=UseColor_ColorTable, specifiedColor=<rgba 0 0 0 0>}
     normalPadding=<ltrb 12 1.5 12 1.5>
     pressed {
      bIsDynamicallyLoaded=false
      drawAs=RoundedBox
      imageType=NoImage
      mirroring=NoMirror
      resourceName=None
      resourceObject=null
      tiling=NoTile
      imageSize=<0 0>
      margin=<ltrb 0 0 0 0>
      outlineSettings {
       bUseBrushTransparency=false
       roundingType=FixedRadius
       width=1
       color {colorUseRule=UseColor_ColorTable, specifiedColor=<rgba 0 0 0 0>}
       cornerRadii=<4 4 4 4>
      }
      tintColor {colorUseRule=UseColor_Specified, specifiedColor=<rgba 0.0284 0.0284 0.0284 1>}
      uVRegion {bIsValid=false, max=<0 0>, min=<0 0>}
     }
     pressedForeground {colorUseRule=UseColor_ColorTable, specifiedColor=<rgba 0 0 0 0>}
     pressedPadding=<ltrb 12 2 12 0>
     pressedSlateSound {resourceObject=null}
    }}
   slot {class=ButtonSlot}
  }
  Image_97 {
   widgetClassPath=&Image
   bIsVariable=true
   parent=FindMeInvalidContainers
   widget=Image_97
   properties {brush {
     bIsDynamicallyLoaded=false
     drawAs=Image
     imageType=NoImage
     mirroring=NoMirror
     resourceName=None
     resourceObject=/Game/Athena/UI/InGame/Creative/PropertyEditor/Icon_Picker/Textures/128x128/T_UI_IconLibrary_tree_128
     tiling=NoTile
     imageSize=<64 64>
     margin=<ltrb 0 0 0 0>
     outlineSettings {
      bUseBrushTransparency=false
      roundingType=HalfHeightRadius
      width=0
      color {colorUseRule=UseColor_Specified, specifiedColor=<rgba 0 0 0 0>}
      cornerRadii=<0 0 0 0>
     }
     tintColor {colorUseRule=UseColor_Specified, specifiedColor=<rgba 1 1 1 1>}
     uVRegion {bIsValid=false, max=<0 0>, min=<0 0>}
    }}
   slot {class=ButtonSlot}
  }
  CommonTextBlock_142 {
   widgetClassPath=&CommonTextBlock
   parent=VerticalBox_0
   widget=CommonTextBlock_142
   properties {
    style=&TextStyle-Base_C
    text=Click Tree Button to Validate ...
    font {
     bForceMonospaced=false
     bMaterialIsStencil=false
     fontMaterial=null
     fontObject=&NotoSans
     letterSpacing=0
     monospacedWidth=1
     size=24
     skewAmount=0
     typefaceFontName=NotoSans-Regular
     outlineSettings {
      bApplyOutlineToDropShadows=false
      bMiteredCorners=false
      bSeparateFillAlpha=false
      outlineMaterial=null
      outlineSize=0
      outlineColor=<rgba 0 0 0 1>
     }
    }
    shadowOffset=<0 0>
   }
   slot {class=VerticalBoxSlot, properties {horizontalAlignment=HAlign_Left}}
  }
  EditableTextBox_0 {
   widgetClassPath=&EditableTextBox
   bIsVariable=true
   parent=VerticalBox_0
   widget=EditableTextBox_0
   properties {
    text=Load all regions for best results - check output log for results.
    widgetStyle {
     backgroundColor {colorUseRule=UseColor_Specified, specifiedColor=<rgba 1 1 1 1>}
     backgroundImageFocused {
      bIsDynamicallyLoaded=false
      drawAs=RoundedBox
      imageType=NoImage
      mirroring=NoMirror
      resourceName=None
      resourceObject=null
      tiling=NoTile
      imageSize=<0 0>
      margin=<ltrb 0 0 0 0>
      outlineSettings {
       bUseBrushTransparency=false
       roundingType=FixedRadius
       width=1
       color {colorUseRule=UseColor_ColorTable, specifiedColor=<rgba 0 0 0 0>}
       cornerRadii=<4 4 4 4>
      }
      tintColor {colorUseRule=UseColor_Specified, specifiedColor=<rgba 0.0048 0.0048 0.0048 1>}
      uVRegion {bIsValid=false, max=<0 0>, min=<0 0>}
     }
     backgroundImageHovered {
      bIsDynamicallyLoaded=false
      drawAs=RoundedBox
      imageType=NoImage
      mirroring=NoMirror
      resourceName=None
      resourceObject=null
      tiling=NoTile
      imageSize=<0 0>
      margin=<ltrb 0 0 0 0>
      outlineSettings {
       bUseBrushTransparency=false
       roundingType=FixedRadius
       width=1
       color {colorUseRule=UseColor_ColorTable, specifiedColor=<rgba 0 0 0 0>}
       cornerRadii=<4 4 4 4>
      }
      tintColor {colorUseRule=UseColor_Specified, specifiedColor=<rgba 0.0048 0.0048 0.0048 1>}
      uVRegion {bIsValid=false, max=<0 0>, min=<0 0>}
     }
     backgroundImageNormal {
      bIsDynamicallyLoaded=false
      drawAs=RoundedBox
      imageType=NoImage
      mirroring=NoMirror
      resourceName=None
      resourceObject=null
      tiling=NoTile
      imageSize=<0 0>
      margin=<ltrb 0 0 0 0>
      outlineSettings {
       bUseBrushTransparency=false
       roundingType=FixedRadius
       width=1
       color {colorUseRule=UseColor_ColorTable, specifiedColor=<rgba 0 0 0 0>}
       cornerRadii=<4 4 4 4>
      }
      tintColor {colorUseRule=UseColor_Specified, specifiedColor=<rgba 0.0048 0.0048 0.0048 1>}
      uVRegion {bIsValid=false, max=<0 0>, min=<0 0>}
     }
     backgroundImageReadOnly {
      bIsDynamicallyLoaded=false
      drawAs=RoundedBox
      imageType=NoImage
      mirroring=NoMirror
      resourceName=None
      resourceObject=null
      tiling=NoTile
      imageSize=<0 0>
      margin=<ltrb 0 0 0 0>
      outlineSettings {
       bUseBrushTransparency=false
       roundingType=FixedRadius
       width=1
       color {colorUseRule=UseColor_ColorTable, specifiedColor=<rgba 0 0 0 0>}
       cornerRadii=<4 4 4 4>
      }
      tintColor {colorUseRule=UseColor_Specified, specifiedColor=<rgba 0.0284 0.0284 0.0284 1>}
      uVRegion {bIsValid=false, max=<0 0>, min=<0 0>}
     }
     focusedForegroundColor {colorUseRule=UseColor_Specified, specifiedColor=<rgba 1 1 1 1>}
     foregroundColor {colorUseRule=UseColor_Specified, specifiedColor=<rgba 0.5271 0.5271 0.5271 1>}
     hScrollBarPadding=<ltrb 0 0 0 0>
     padding=<ltrb 8 3 8 4>
     readOnlyForegroundColor {colorUseRule=UseColor_Foreground, specifiedColor=<rgba 1 0 1 1>}
     scrollBarStyle {
      thickness=8
      draggedThumbImage {
       bIsDynamicallyLoaded=false
       drawAs=RoundedBox
       imageType=NoImage
       mirroring=NoMirror
       resourceName=None
       resourceObject=null
       tiling=NoTile
       imageSize=<0 0>
       margin=<ltrb 0 0 0 0>
       outlineSettings {
        bUseBrushTransparency=false
        roundingType=FixedRadius
        width=0
        color {colorUseRule=UseColor_Specified, specifiedColor=<rgba 0 0 0 0>}
        cornerRadii=<4 4 4 4>
       }
       tintColor {colorUseRule=UseColor_Specified, specifiedColor=<rgba 0.2159 0.2159 0.2159 1>}
       uVRegion {bIsValid=false, max=<0 0>, min=<0 0>}
      }
      horizontalBackgroundImage {
       bIsDynamicallyLoaded=false
       drawAs=NoDrawType
       imageType=NoImage
       mirroring=NoMirror
       resourceName=None
       resourceObject=null
       tiling=NoTile
       imageSize=<0 0>
       margin=<ltrb 0 0 0 0>
       outlineSettings {
        bUseBrushTransparency=false
        roundingType=HalfHeightRadius
        width=0
        color {colorUseRule=UseColor_Specified, specifiedColor=<rgba 0 0 0 0>}
        cornerRadii=<0 0 0 0>
       }
       tintColor {colorUseRule=UseColor_Specified, specifiedColor=<rgba 1 1 1 1>}
       uVRegion {bIsValid=false, max=<0 0>, min=<0 0>}
      }
      horizontalBottomSlotImage {
       bIsDynamicallyLoaded=false
       drawAs=NoDrawType
       imageType=NoImage
       mirroring=NoMirror
       resourceName=None
       resourceObject=null
       tiling=NoTile
       imageSize=<0 0>
       margin=<ltrb 0 0 0 0>
       outlineSettings {
        bUseBrushTransparency=false
        roundingType=HalfHeightRadius
        width=0
        color {colorUseRule=UseColor_Specified, specifiedColor=<rgba 0 0 0 0>}
        cornerRadii=<0 0 0 0>
       }
       tintColor {colorUseRule=UseColor_Specified, specifiedColor=<rgba 1 1 1 1>}
       uVRegion {bIsValid=false, max=<0 0>, min=<0 0>}
      }
      horizontalTopSlotImage {
       bIsDynamicallyLoaded=false
       drawAs=NoDrawType
       imageType=NoImage
       mirroring=NoMirror
       resourceName=None
       resourceObject=null
       tiling=NoTile
       imageSize=<0 0>
       margin=<ltrb 0 0 0 0>
       outlineSettings {
        bUseBrushTransparency=false
        roundingType=HalfHeightRadius
        width=0
        color {colorUseRule=UseColor_Specified, specifiedColor=<rgba 0 0 0 0>}
        cornerRadii=<0 0 0 0>
       }
       tintColor {colorUseRule=UseColor_Specified, specifiedColor=<rgba 1 1 1 1>}
       uVRegion {bIsValid=false, max=<0 0>, min=<0 0>}
      }
      hoveredThumbImage {
       bIsDynamicallyLoaded=false
       drawAs=RoundedBox
       imageType=NoImage
       mirroring=NoMirror
       resourceName=None
       resourceObject=null
       tiling=NoTile
       imageSize=<0 0>
       margin=<ltrb 0 0 0 0>
       outlineSettings {
        bUseBrushTransparency=false
        roundingType=FixedRadius
        width=0
        color {colorUseRule=UseColor_Specified, specifiedColor=<rgba 0 0 0 0>}
        cornerRadii=<4 4 4 4>
       }
       tintColor {colorUseRule=UseColor_Specified, specifiedColor=<rgba 0.2159 0.2159 0.2159 1>}
       uVRegion {bIsValid=false, max=<0 0>, min=<0 0>}
      }
      normalThumbImage {
       bIsDynamicallyLoaded=false
       drawAs=RoundedBox
       imageType=NoImage
       mirroring=NoMirror
       resourceName=None
       resourceObject=null
       tiling=NoTile
       imageSize=<0 0>
       margin=<ltrb 0 0 0 0>
       outlineSettings {
        bUseBrushTransparency=false
        roundingType=FixedRadius
        width=0
        color {colorUseRule=UseColor_Specified, specifiedColor=<rgba 0 0 0 0>}
        cornerRadii=<4 4 4 4>
       }
       tintColor {colorUseRule=UseColor_Specified, specifiedColor=<rgba 0.0953 0.0953 0.0953 1>}
       uVRegion {bIsValid=false, max=<0 0>, min=<0 0>}
      }
      verticalBackgroundImage {
       bIsDynamicallyLoaded=false
       drawAs=NoDrawType
       imageType=NoImage
       mirroring=NoMirror
       resourceName=None
       resourceObject=null
       tiling=NoTile
       imageSize=<0 0>
       margin=<ltrb 0 0 0 0>
       outlineSettings {
        bUseBrushTransparency=false
        roundingType=HalfHeightRadius
        width=0
        color {colorUseRule=UseColor_Specified, specifiedColor=<rgba 0 0 0 0>}
        cornerRadii=<0 0 0 0>
       }
       tintColor {colorUseRule=UseColor_Specified, specifiedColor=<rgba 1 1 1 1>}
       uVRegion {bIsValid=false, max=<0 0>, min=<0 0>}
      }
      verticalBottomSlotImage {
       bIsDynamicallyLoaded=false
       drawAs=NoDrawType
       imageType=NoImage
       mirroring=NoMirror
       resourceName=None
       resourceObject=null
       tiling=NoTile
       imageSize=<0 0>
       margin=<ltrb 0 0 0 0>
       outlineSettings {
        bUseBrushTransparency=false
        roundingType=HalfHeightRadius
        width=0
        color {colorUseRule=UseColor_Specified, specifiedColor=<rgba 0 0 0 0>}
        cornerRadii=<0 0 0 0>
       }
       tintColor {colorUseRule=UseColor_Specified, specifiedColor=<rgba 1 1 1 1>}
       uVRegion {bIsValid=false, max=<0 0>, min=<0 0>}
      }
      verticalTopSlotImage {
       bIsDynamicallyLoaded=false
       drawAs=NoDrawType
       imageType=NoImage
       mirroring=NoMirror
       resourceName=None
       resourceObject=null
       tiling=NoTile
       imageSize=<0 0>
       margin=<ltrb 0 0 0 0>
       outlineSettings {
        bUseBrushTransparency=false
        roundingType=HalfHeightRadius
        width=0
        color {colorUseRule=UseColor_Specified, specifiedColor=<rgba 0 0 0 0>}
        cornerRadii=<0 0 0 0>
       }
       tintColor {colorUseRule=UseColor_Specified, specifiedColor=<rgba 1 1 1 1>}
       uVRegion {bIsValid=false, max=<0 0>, min=<0 0>}
      }
     }
     textStyle {
      overflowPolicy=Clip
      transformPolicy=None
      colorAndOpacity {colorUseRule=UseColor_Foreground, specifiedColor=<rgba 1 0 1 1>}
      font {
       bForceMonospaced=false
       bMaterialIsStencil=false
       fontMaterial=null
       fontObject=null
       letterSpacing=0
       monospacedWidth=1
       size=10
       skewAmount=0
       typefaceFontName=Regular
       outlineSettings {
        bApplyOutlineToDropShadows=false
        bMiteredCorners=false
        bSeparateFillAlpha=false
        outlineMaterial=null
        outlineSize=0
        outlineColor=<rgba 0 0 0 1>
       }
      }
      highlightColor {colorUseRule=UseColor_ColorTable, specifiedColor=<rgba 0 0 0 0>}
      highlightShape {
       bIsDynamicallyLoaded=false
       drawAs=Image
       imageType=NoImage
       mirroring=NoMirror
       resourceName=None
       resourceObject=null
       tiling=NoTile
       imageSize=<0 0>
       margin=<ltrb 0 0 0 0>
       outlineSettings {
        bUseBrushTransparency=false
        roundingType=HalfHeightRadius
        width=0
        color {colorUseRule=UseColor_Specified, specifiedColor=<rgba 0 0 0 0>}
        cornerRadii=<0 0 0 0>
       }
       tintColor {colorUseRule=UseColor_ColorTable, specifiedColor=<rgba 0 0 0 0>}
       uVRegion {bIsValid=false, max=<0 0>, min=<0 0>}
      }
      selectedBackgroundColor {colorUseRule=UseColor_ColorTable, specifiedColor=<rgba 0 0 0 0>}
      shadowColorAndOpacity=<rgba 0 0 0 1>
      shadowOffset=<0 0>
      strikeBrush {
       bIsDynamicallyLoaded=false
       drawAs=Image
       imageType=NoImage
       mirroring=NoMirror
       resourceName=None
       resourceObject=null
       tiling=NoTile
       imageSize=<32 32>
       margin=<ltrb 0 0 0 0>
       outlineSettings {
        bUseBrushTransparency=false
        roundingType=HalfHeightRadius
        width=0
        color {colorUseRule=UseColor_Specified, specifiedColor=<rgba 0 0 0 0>}
        cornerRadii=<0 0 0 0>
       }
       tintColor {colorUseRule=UseColor_Specified, specifiedColor=<rgba 1 1 1 1>}
       uVRegion {bIsValid=false, max=<0 0>, min=<0 0>}
      }
      underlineBrush {
       bIsDynamicallyLoaded=false
       drawAs=Image
       imageType=NoImage
       mirroring=NoMirror
       resourceName=None
       resourceObject=null
       tiling=NoTile
       imageSize=<32 32>
       margin=<ltrb 0 0 0 0>
       outlineSettings {
        bUseBrushTransparency=false
        roundingType=HalfHeightRadius
        width=0
        color {colorUseRule=UseColor_Specified, specifiedColor=<rgba 0 0 0 0>}
        cornerRadii=<0 0 0 0>
       }
       tintColor {colorUseRule=UseColor_Specified, specifiedColor=<rgba 1 1 1 1>}
       uVRegion {bIsValid=false, max=<0 0>, min=<0 0>}
      }
     }
     vScrollBarPadding=<ltrb 0 0 0 0>
    }
   }
   slot {class=VerticalBoxSlot}
  }
 ]
}
